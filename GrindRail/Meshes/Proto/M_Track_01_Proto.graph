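# Prototype Default Lit material, apparently for grind rail track meshes, judging by its name and path.
# It holds only two parameters, a scalar Roughness and a dark blue-grey vector color named "COlor", with no textures or math nodes.
# Usage is enabled for spline meshes and instanced static meshes, with automatic usage setting in the editor turned off.
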Material /GrindRail/Meshes/Proto/M_Track_01_Proto
graph {
 properties {bAutomaticallySetUsageInEditor=false, bUsedWithInstancedStaticMeshes=true, bUsedWithSplineMeshes=true, usedShadingModels=Default Lit}
 nodes [
  ScalarParameter() {defaultValue=0.25, parameterName=Roughness}
  VectorParameter() {parameterName=COlor, defaultValue=<rgba 0.0312 0.0323 0.0469 1>}
 ]
}
